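; BP_RibbonSparkles is a Blueprint actor in the GoldRibbons tech art folder that shows a configurable material on a plane, apparently as a ribbon sparkle effect.
; Its construction script creates a dynamic material instance from the Material variable, assigns it to the Plane's first slot, and feeds the Scale variable into the WPOScale parameter.
; BeginPlay, Tick and overlap events are empty, so any motion is left to the material.
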(event EventBeginPlay)

(event Collision|EventActorBeginOverlap
  (params (OtherActor :type Actor%20Object%20Reference)))

(event EventTick
  (params (DeltaSeconds :type Float%20%28single-precision%29)))

(fn ConstructionScript
  (bind _returnvalue (Rendering|Material|CreateDynamicMaterialInstance (Variables|Default|GetMaterial) "None" "None" :!class /Script/Engine.KismetMaterialLibrary))
  (Rendering|Material|SetMaterial (Variables|Default|GetPlane) 0 _returnvalue)
  (Rendering|Material|SetScalarParameterValue _returnvalue "WPOScale" (Variables|Default|GetScale) :!class /Script/Engine.MaterialInstanceDynamic))
