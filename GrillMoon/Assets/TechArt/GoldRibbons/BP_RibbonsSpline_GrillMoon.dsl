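; Spline-driven ribbon Blueprint for GrillMoon that builds a twisting chain of spline meshes with an optional end-card plane.
; Construction spawns one mesh per segment on a shared dynamic material, applying per-point scale and per-segment up-direction overrides from maps.
; Custom primitive data carries segment length, thickness, end-card flag and gradient fractions forming one gradient along the ribbon.
; BeginPlay reapplies that data on non-authority instances, working around it not sticking in builds.
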(event EventBeginPlay
  ; Custom Prim Data doesnt seem to be setting correctly in Build. Attempting to set here
  (Networking|SwitchHasAuthority
    (:Remote
      (CallFunction|ApplyCustomPrimDatatoMeshComponents))))

(fn AddSplinePointThicknessIndex
  (Utilities|Map|Add (Variables|Overrrides|GetSplinePointScaleMultiplier) 999 1.0))

(fn AddSplineUpDirectionOverride
  (Utilities|Map|Add (Variables|Overrrides|GetSplineUpDirOverride) 999 "0.000000,0.000000,1.000000"))

(fn ApplyCustomPrimDataToMeshComponents
  (bind _spline (Variables|Default|GetSpline))
  (bind _returnvalue (Spline|GetNumberOfSplinePoints _spline))
  (bind _returnvalue_2 (Math|Conversions|ToFloat(Integer) _returnvalue))
  (if (Utilities|Array|IsNotEmpty (Variables|Default|GetSplineMeshComponents))
    (for _index (range (+ (- (Spline|GetNumberOfSplinePoints (Variables|Default|GetSpline)) 2) 1))
      ;[0 0 0] Mesh Component Setup
      (LocalVariables|SetSMC (Utilities|Array|Get(acopy) (Variables|Default|GetSplineMeshComponents) _index))
      ;;[1 0.3395 0.01481] Thickness Multipliers Per Point overrides
      (Rendering|Material|SetCustomPrimitiveDataFloat (LocalVariables|GetSMC) 10 (select (.ReturnValue (Utilities|Map|Find (Variables|Overrrides|GetSplinePointScaleMultiplier) _index)) (* (Variables|Spline|GetScale) (Utilities|Map|Find (Variables|Overrrides|GetSplinePointScaleMultiplier) _index)) (Variables|Spline|GetScale)))
      (Rendering|Material|SetCustomPrimitiveDataFloat (LocalVariables|GetSMC) 11 (select (.ReturnValue (Utilities|Map|Find (Variables|Overrrides|GetSplinePointScaleMultiplier) (+ _index 1))) (* (Variables|Spline|GetScale) (Utilities|Map|Find (Variables|Overrrides|GetSplinePointScaleMultiplier) (+ _index 1))) (Variables|Spline|GetScale)))
      ;[0 0 0] Gradient for each spline component to form one master gradient
      ;; End
      (Rendering|Material|SetCustomPrimitiveDataFloat (LocalVariables|GetSMC) 2 (/ (Math|Conversions|ToFloat(Integer) (+ _index 1)) _returnvalue_2))
      ;; Start
      (Rendering|Material|SetCustomPrimitiveDataFloat (LocalVariables|GetSMC) 1 (select (== _index 0) 0.0 (/ (Math|Conversions|ToFloat(Integer) _index) _returnvalue_2)))
      (Rendering|Material|SetCustomPrimitiveDataFloat (LocalVariables|GetSMC) 3 (Math|Vector|Distance(Vector) (Spline|GetLocationatSplinePoint (Variables|Default|GetSpline) _index "World") (Spline|GetLocationatSplinePoint (Variables|Default|GetSpline) (+ _index 1) "World")))
      (Rendering|Material|SetCustomPrimitiveDataFloat (LocalVariables|GetSMC) 5 (Math|Float|SelectFloat 1.0 0.0 (Variables|SplineEnd|GetbUseEndSplineCard))))
    (return)
    (else
      (Development|PrintString "FAILED TO APPLY DATA, NO COMPONENTS"))))

(fn ConstructSplineMesh
  (bind _spline (Variables|Default|GetSpline))
  (bind _returnvalue (Spline|GetNumberOfSplinePoints _spline))
  (bind _spline_2 (Variables|Default|GetSpline))
  (bind _returnvalue_2 (Spline|GetNumberOfSplinePoints _spline_2))
  (bind _returnvalue_3 (Math|Conversions|ToFloat(Integer) _returnvalue_2))
  (bind _returnvalue_4 (Math|Conversions|ToFloat(Integer) _returnvalue))
  (bind _splinepointscalemultiplier (Variables|Overrrides|GetSplinePointScaleMultiplier))
  (bind _splinepointscalemultiplier_2 (Variables|Overrrides|GetSplinePointScaleMultiplier))
  (bind _scale (Variables|Spline|GetScale))
  (bind _smc (LocalVariables|GetSMC))
  (bind _scale_2 (Variables|Spline|GetScale))
  (bind _smc_2 (LocalVariables|GetSMC))
  (Utilities|Array|Clear (Variables|Default|GetSplineMeshComponents))
  (Utilities|IsValid (Variables|MeshComponent|GetStaticMesh)
    (:Is%20Valid
      (bind _returnvalue_5 (Rendering|Material|CreateDynamicMaterialInstance (Variables|MeshComponent|GetMaterial) "None" "None" :!class /Script/Engine.KismetMaterialLibrary))
      (bind _output_get (Variables|MeshComponent|SetMID _returnvalue_5))
      ;[0 0 0] Set Start and End Point in the Material in World Space
      (Rendering|Material|SetVectorParameterValue _output_get "BP.StartPoint" (Math|Conversions|ToLinearColor(Vector) (Spline|GetLocationatSplinePoint (Variables|Default|GetSpline) 0 "World")) :!class /Script/Engine.MaterialInstanceDynamic)
      (Rendering|Material|SetVectorParameterValue _output_get "BP.EndPoint" (Math|Conversions|ToLinearColor(Vector) (Spline|GetLocationatSplinePoint (Variables|Default|GetSpline) (Spline|GetNumberOfSplinePoints (Variables|Default|GetSpline)) "World")) :!class /Script/Engine.MaterialInstanceDynamic)
      (for _index (range (+ (- (Spline|GetNumberOfSplinePoints (Variables|Default|GetSpline)) 2) 1))
        ;[0 0 0] Mesh Component Setup
        (bind _returnvalue_6 (AddComponent|NODE_AddSplineMeshComponent-0 :bManualAttachment true))
        (LocalVariables|SetSMC _returnvalue_6)
        (Utilities|Array|Add (Variables|Default|GetSplineMeshComponents) _returnvalue_6)
        (Components|StaticMesh|SetStaticMesh (LocalVariables|GetSMC) (Variables|MeshComponent|GetStaticMesh))
        (Rendering|Material|SetMaterial (LocalVariables|GetSMC) 0 (Variables|MeshComponent|GetMID))
        ;;[1 0.3395 0.01481] Per Component Overrides for Spline Dir to help fix twisting - Maybe be able to remove later
        (SplineMesh|SetSplineUpDir (LocalVariables|GetSMC) (Math|Vector|Normalize (select (.ReturnValue (Utilities|Map|Find (Variables|Overrrides|GetSplineUpDirOverride) _index)) (Utilities|Map|Find (Variables|Overrrides|GetSplineUpDirOverride) _index) (Variables|Overrrides|GetSplineUpDirectionDefault))) false)
        ;;[1 0.3395 0.01481] Thickness Multipliers Per Point overrides
        (SplineMesh|SetStartScale _smc (Math|Conversions|ToVector2D (select (.ReturnValue (Utilities|Map|Find _splinepointscalemultiplier_2 _index)) (* _scale_2 (Utilities|Map|Find _splinepointscalemultiplier_2 _index)) _scale_2)) false :!class /Script/Engine.SplineMeshComponent)
        (Rendering|Material|SetCustomPrimitiveDataFloat _smc 10 (select (.ReturnValue (Utilities|Map|Find _splinepointscalemultiplier_2 _index)) (* _scale_2 (Utilities|Map|Find _splinepointscalemultiplier_2 _index)) _scale_2))
        (SplineMesh|SetEndScale _smc_2 (Math|Conversions|ToVector2D (select (.ReturnValue (Utilities|Map|Find _splinepointscalemultiplier (+ _index 1))) (* _scale (Utilities|Map|Find _splinepointscalemultiplier (+ _index 1))) _scale)) false :!class /Script/Engine.SplineMeshComponent)
        (Rendering|Material|SetCustomPrimitiveDataFloat _smc_2 11 (select (.ReturnValue (Utilities|Map|Find _splinepointscalemultiplier (+ _index 1))) (* _scale (Utilities|Map|Find _splinepointscalemultiplier (+ _index 1))) _scale))
        (SplineMesh|SetForwardAxis (LocalVariables|GetSMC) (Variables|Spline|GetForwardAxis) false)
        (SplineMesh|SetStartandEnd (LocalVariables|GetSMC) (Spline|GetLocationandTangentatSplinePoint (Variables|Default|GetSpline) _index "Local") (.Tangent (Spline|GetLocationandTangentatSplinePoint (Variables|Default|GetSpline) _index "Local")) (Spline|GetLocationandTangentatSplinePoint (Variables|Default|GetSpline) (+ _index 1) "Local") (.Tangent (Spline|GetLocationandTangentatSplinePoint (Variables|Default|GetSpline) (+ _index 1) "Local")) true :!class /Script/Engine.SplineMeshComponent)
        (Transformation|AttachComponentToComponent (LocalVariables|GetSMC) (Variables|Transformation|GetRootComponent) "None" "KeepRelative" "KeepWorld" "KeepWorld" false)
        (SplineMesh|SetEndRollDegrees (LocalVariables|GetSMC) (* (Variables|Spline|GetNumberOfTwistsInDegrees) (/ (Math|Conversions|ToFloat(Integer) (+ _index 1)) _returnvalue_3)) false)
        (SplineMesh|SetStartRollDegrees (LocalVariables|GetSMC) (* (Variables|Spline|GetNumberOfTwistsInDegrees) (select (== _index 0) 0.0 (/ (Math|Conversions|ToFloat(Integer) _index) _returnvalue_3))) false)
        ;[0 0 0] Gradient for each spline component to form one master gradient
        ;; End
        (Rendering|Material|SetCustomPrimitiveDataFloat (LocalVariables|GetSMC) 2 (/ (Math|Conversions|ToFloat(Integer) (+ _index 1)) _returnvalue_4))
        ;; Start
        (Rendering|Material|SetCustomPrimitiveDataFloat (LocalVariables|GetSMC) 1 (select (== _index 0) 0.0 (/ (Math|Conversions|ToFloat(Integer) _index) _returnvalue_4)))
        (Rendering|Material|SetCustomPrimitiveDataFloat (LocalVariables|GetSMC) 3 (Math|Vector|Distance(Vector) (Spline|GetLocationatSplinePoint (Variables|Default|GetSpline) _index "World") (Spline|GetLocationatSplinePoint (Variables|Default|GetSpline) (+ _index 1) "World")))
        (Rendering|Material|SetCustomPrimitiveDataFloat (LocalVariables|GetSMC) 5 (Math|Float|SelectFloat 1.0 0.0 (Variables|SplineEnd|GetbUseEndSplineCard))))
      ; EndCard
      (if (Variables|SplineEnd|GetbUseEndSplineCard)
        (Transformation|SetWorldLocation (Variables|Default|GetEndPlaneGlow) (+ (* (Math|Vector|Normalize (- (Spline|GetLocationatSplinePoint (Variables|Default|GetSpline) (- (Spline|GetNumberOfSplinePoints (Variables|Default|GetSpline)) 1) "World") (Spline|GetLocationatSplinePoint (Variables|Default|GetSpline) (- (Spline|GetNumberOfSplinePoints (Variables|Default|GetSpline)) 2) "World"))) (Variables|SplineEnd|GetSplineEndDirectionalOffset)) (+ (Spline|GetLocationatSplinePoint (Variables|Default|GetSpline) (- (Spline|GetNumberOfSplinePoints (Variables|Default|GetSpline)) 1) "World") (Variables|SplineEnd|GetSplineEndCardOffset))) false false)
        (Rendering|Material|SetMaterial (Variables|Default|GetEndPlaneGlow) 0 (Variables|SplineEnd|GetSplineEndMaterial))
        (else
          (Components|DestroyComponent (Variables|Default|GetEndPlaneGlow)))))))

(fn ConstructionScript
  (CallFunction|ConstructSplineMesh))
