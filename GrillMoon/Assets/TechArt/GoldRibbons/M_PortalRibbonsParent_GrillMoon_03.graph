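# Parent additive, unlit, two-sided material for gold portal ribbons on spline meshes, driven by per-segment custom primitive data and endpoint parameters.
# Warped, panning noise layers, spline color gradients, fresnel, sparkle and pulse terms build the glow, with tip, depth and camera-distance fades.
# Summed sine waves wobble vertices by camera distance.
# Vertex color blue optionally separates a normal-inflated airglow shell from the core, and a static switch provides a debug view.
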Material /GrillMoon/Assets/TechArt/GoldRibbons/M_PortalRibbonsParent_GrillMoon_03
graph {
 properties {
  bAutomaticallySetUsageInEditor=false
  bEnableResponsiveAA=true
  bTangentSpaceNormal=false
  bUseMaterialAttributes=true
  bUsedWithSplineMeshes=true
  blendMode=BLEND_Additive
  numCustomizedUVs=1
  shadingModel=MSM_Unlit
  twoSided=true
  usedShadingModels=Unlit
 }
 nodes [
  n1 = VertexNormalWS()
  n2 = Add(A=NamedRerouteUsage(), B=Multiply(A=NamedRerouteUsage(), B=ScalarParameter() {defaultValue=0.2, group=Pulse, parameterName=Pulse.Speed}))
  n3 = Power(Base=Saturate(None=Sine(None=@n2)), Exp=ScalarParameter() {defaultValue=1, group=Pulse, parameterName=Pulse.Width})
  n4 = Multiply(A=@n3, B=ScalarParameter() {defaultValue=50, group=Pulse, parameterName=Pulse.WPO.Intensity})
  Add(B=LinearInterpolate(A=Multiply(A=@n1, B=@n4), Alpha=NamedRerouteUsage()) {constB=0, desc=Dont pulse at tips}) {desc=WPO}
  n5 = WorldPosition()
  n6 = SphereMask(A=@n5.XYZ, B=CameraPositionWS()) {attenuationRadius=150, hardnessPercent=10}
  n7 = ScalarParameter() {group=General, parameterName=General.NearCameraFade.MinVal}
  n8 = ScalarParameter() {defaultValue=1, group=General, parameterName=General.NearCameraFade.MaxVal}
  n9 = CameraPositionWS()
  n10 = ScalarParameter() {defaultValue=200, group=General, parameterName=General.NearCameraFade.Distance}
  n11 = ScalarParameter() {defaultValue=0.8, group=General, parameterName=General.NearCameraFade.Hardness}
  n12 = StaticSwitchParameter(True=OneMinus(None=SphereMask(A=@n5.XYZ, B=@n9, Radius=@n10, Hardness=@n11)), False=Constant() {r=1}) {defaultValue=true, parameterName=UseNearCameraFade}
  NamedRerouteDeclaration(None=Multiply(A=OneMinus(None=@n6), B=LinearInterpolate(A=@n7, B=@n8, Alpha=@n12))) {name=NearCameraFade.Mask, nodeColor=<rgba 0 0.25 1 1>}
  n13 = TextureObject() {samplerType=SAMPLERTYPE_LinearColor, texture=/Game/Global/Materials/Character_Parent_2023/Anime_Parent_2023/Textures/T_Hatching_Voronoi}
  n14 = MaterialFunctionCall(Float3=MaterialFunctionCall(TextureObject=@n13, TextureSize=ScalarParameter() {defaultValue=512, parameterName=Distortion.Tiling}) {materialFunction=&WorldAlignedTexture}.XYZ Texture) {materialFunction=&BreakOutFloat3Components}
  n15 = MaterialFunctionCall(X=OneMinus(None=@n14.G), Y=@n14.G) {materialFunction=&MakeFloat2}
  n16 = ScalarParameter() {defaultValue=0.01, parameterName=Distortion.Intensity.Min}
  n17 = ScalarParameter() {defaultValue=0.01, parameterName=Distortion.Intensity.Max}
  n18 = TextureObject() {samplerType=SAMPLERTYPE_Masks, texture=/Game/Global/Textures/Default/Noise/T_Generic_Perlin}
  n19 = ScalarParameter() {defaultValue=512, parameterName=Distortion.VariationSize}
  n20 = Add(A=Multiply(A=NamedRerouteUsage(), B=ScalarParameter() {defaultValue=20, parameterName=Distortion.VariationSpeed}), B=WorldPosition().XYZ)
  n21 = Power(
   Base=ComponentMask(None=MaterialFunctionCall(TextureObject=@n18, TextureSize=@n19, WorldPosition=@n20) {materialFunction=&WorldAlignedTexture}.XYZ Texture) {g=true}
   Exp=ScalarParameter() {defaultValue=1, parameterName=Distortion.Intensity.Contrast}
  )
  NamedRerouteDeclaration(None=Multiply(A=ConstantBiasScale(None=@n15.Result) {bias=-0.5, scale=2}, B=LinearInterpolate(A=@n16, B=@n17, Alpha=@n21))) {name=DistortionUVs, nodeColor=<rgba 1 0 0.1172 1>}
  n22 = OneMinus(None=VertexColor().B)
  n23 = LinearInterpolate(
   A=ScalarParameter() {defaultValue=200, group=General, parameterName=Airglow.Tip.FadeRadius, sortPriority=50}
   B=ScalarParameter() {defaultValue=200, group=General, parameterName=Tip.FadeRadius, sortPriority=50}
   Alpha=@n22
  )
  n24 = LinearInterpolate(
   A=ScalarParameter() {defaultValue=0.6, group=General, parameterName=Airglow.Tip.Hardness, sortPriority=50}
   B=ScalarParameter() {defaultValue=0.7, group=General, parameterName=Tip.FadeHardness, sortPriority=50}
   Alpha=@n22
  )
  n25 = WorldPosition() {worldPositionShaderOffset=WPT_ExcludeAllShaderOffsets}
  n26 = SphereMask(
   A=@n25.XYZ
   B=VectorParameter() {parameterName=BP.StartPoint, primitiveDataIndex=1, defaultValue=<rgba 0 0 0 1>}.RGB
   Radius=@n23
   Hardness=@n24
  )
  n27 = SphereMask(
   A=@n25.XYZ
   B=VectorParameter() {parameterName=BP.EndPoint, primitiveDataIndex=2, defaultValue=<rgba 0 0 0 1>}.RGB
   Radius=@n23
   Hardness=@n24
  )
  NamedRerouteDeclaration(None=Max(A=@n26, B=@n27)) {name=Tip.Mask, nodeColor=<rgba 0.5469 1 0 1>}
  NamedRerouteDeclaration(None=AppendVector(
    A=ScalarParameter() {bUseCustomPrimitiveData=true, defaultValue=1, parameterName=CustomPrim_RadiusStartMulti, primitiveDataIndex=10}
    B=ScalarParameter() {bUseCustomPrimitiveData=true, defaultValue=1, parameterName=CustomPrim_RadiusEndMulti, primitiveDataIndex=11}
   )) {name=CustomPrimRadiusStartEndV2, nodeColor=<rgba 0.8047 1 0 1>}
  n28 = Multiply(A=@n3, B=VectorParameter() {group=Color, parameterName=Color.Pulse, defaultValue=<rgba 1 0.8219 0.4583 1>}.RGB)
  NamedRerouteDeclaration(None=@n28) {name=Pulse.Emissive, nodeColor=<rgba 1 0 0.1875 1>}
  NamedRerouteDeclaration(None=TextureObject() {texture=&/T_NoisePack_01_GrillMoon}) {name=PackedNoise.TexObj, nodeColor=<rgba 1 0 0.6797 1>}
  NamedRerouteDeclaration(None=Divide(
    A=ScalarParameter() {bUseCustomPrimitiveData=true, defaultValue=100, parameterName=CustomPrim_SegLen, primitiveDataIndex=3}
    B=ScalarParameter() {defaultValue=100, parameterName=General.BaseLength}
   )) {name=SegmentLength, nodeColor=<rgba 0.7578 0 1 1>}
  n29 = NamedRerouteUsage()
  n30 = TextureCoordinate()
  n31 = MaterialFunctionCall(Float2=@n30) {materialFunction=&BreakOutFloat2Components}
  NamedRerouteDeclaration(None=StaticSwitchParameter(
    True=MaterialFunctionCall(X=@n29, Y=@n31.G) {materialFunction=&MakeFloat2}.Result
    False=@n30
   ) {group=General, parameterName=UseSplineUVCoords, sortPriority=1}) {name=SplineUV, nodeColor=<rgba 0.8281 0 1 1>}
  NamedRerouteDeclaration(None=TextureObject() {samplerType=SAMPLERTYPE_LinearColor, texture=/Game/TimeOfDay/Skydome/PackedEclipseNoise_M}) {name=Noise.TexObj, nodeColor=<rgba 0 0.6719 1 1>}
  n32 = WorldPosition()
  n33 = CameraPositionWS()
  n34 = VectorParameter() {
   group=Sparkles
   parameterName=PerChannelNearFade
   sortPriority=50
   channelNames {a=Hardness, b="", g="", r=""}
   defaultValue=<rgba 500 150 0 50>
  }
  n35 = TextureSample(Tex=NamedRerouteUsage())
  n36 = ScalarParameter() {group=Sparkles, parameterName=PerChannelFadeVariation, sortPriority=50}
  n37 = MaterialFunctionCall(Float3=Add(A=@n34, B=Multiply(A=@n35.RGB, B=@n36))) {materialFunction=&BreakOutFloat3Components}
  n38 = SphereMask(A=@n32.XYZ, B=@n33, Radius=@n37.R, Hardness=@n34.Hardness) {attenuationRadius=1000, hardnessPercent=0.5}
  n39 = SphereMask(A=@n32.XYZ, B=@n33, Radius=@n37.G, Hardness=@n34.Hardness) {attenuationRadius=1000, hardnessPercent=0.5}
  n40 = SphereMask(A=@n32.XYZ, B=@n33, Radius=@n37.B, Hardness=@n34.Hardness) {attenuationRadius=1000, hardnessPercent=0.5}
  NamedRerouteDeclaration(None=MaterialFunctionCall(X=OneMinus(None=@n38), Y=OneMinus(None=@n39), Z=OneMinus(None=@n40)) {materialFunction=&MakeFloat3}.Result) {name=PackedDistanceFades, nodeColor=<rgba 0.2656 0 1 1>}
  n41 = Multiply(
   A=TextureCoordinate()
   B=MaterialFunctionCall(
    X=ScalarParameter() {defaultValue=1, group=Sparkles, parameterName=Sparkles.Tiling.X}
    Y=ScalarParameter() {defaultValue=1, group=Sparkles, parameterName=Sparkles.Tiling.Y}
   ) {materialFunction=&MakeFloat2}.Result
  )
  n42 = Multiply(A=@n41)
  n43 = MaterialFunctionCall(X=Multiply(A=NamedRerouteUsage(), B=ScalarParameter() {defaultValue=0.1, group=Sparkles, parameterName=Sparkles.PanSpeed}) {constB=0.2}) {materialFunction=&MakeFloat2}
  n44 = TextureObject() {samplerType=SAMPLERTYPE_LinearGrayscale, texture=/Game/Effects/Fort_Effects/Textures/Tile/T_WhiteNoise_02_M}
  n45 = TextureSample(UVs=Add(A=@n42, B=@n43.Result), Tex=@n44) {samplerType=SAMPLERTYPE_LinearGrayscale, texture=&T_WhiteNoise_02_M}
  n46 = Constant2Vector() {g=-1, r=1.05}
  n47 = TextureSample(UVs=Add(A=Multiply(A=@n42) {constB=0.75}, B=Multiply(A=@n43.Result, B=@n46)), Tex=@n44) {samplerType=SAMPLERTYPE_LinearGrayscale, texture=&T_WhiteNoise_02_M}
  n48 = Multiply(A=Power(Base=Multiply(A=@n45.R, B=@n47.R)) {constExponent=4}) {constB=2}
  n49 = MaterialFunctionCall(Float3=NamedRerouteUsage()) {materialFunction=&BreakOutFloat3Components}
  n50 = Multiply(A=Round(None=Saturate(None=@n48)), B=@n49.R)
  n51 = Multiply(A=@n41) {constB=2}
  n52 = TextureSample(UVs=Add(A=@n51, B=@n43.Result), Tex=@n44) {samplerType=SAMPLERTYPE_LinearGrayscale, texture=&T_WhiteNoise_02_M}
  n53 = Constant2Vector() {g=1, r=0.1}
  n54 = TextureSample(UVs=Add(A=Multiply(A=@n51) {constB=0.66}, B=Multiply(A=@n43.Result, B=@n53)), Tex=@n44) {samplerType=SAMPLERTYPE_LinearGrayscale, texture=&T_WhiteNoise_02_M}
  n55 = Multiply(A=Power(Base=Multiply(A=@n52.R, B=@n54.R)) {constExponent=4}) {constB=2}
  n56 = Multiply(A=@n49.G, B=Round(None=Saturate(None=@n55)))
  NamedRerouteDeclaration(None=Multiply(A=NamedRerouteUsage(), B=Max(A=@n50, B=@n56))) {name=SparkleColor, nodeColor=<rgba 1 0 0.4453 1>}
  n57 = TextureCoordinate()
  n58 = Saturate(None=LinearInterpolate(
    A=ScalarParameter() {bUseCustomPrimitiveData=true, defaultValue=1, parameterName=CustomPrim_ValEnd, primitiveDataIndex=2}
    B=ScalarParameter() {bUseCustomPrimitiveData=true, parameterName=CustomPrim_ValStart, primitiveDataIndex=1}
    Alpha=ComponentMask(None=@n57) {r=true}
   ))
  NamedRerouteDeclaration(None=StaticSwitchParameter(True=OneMinus(None=@n58), False=@n58) {parameterName=General.InvertSplineGradient}) {name=SplineGradient, nodeColor=<rgba 0 1 0.6016 1>}
  n59 = CameraPositionWS()
  n60 = WorldPosition()
  n61 = Divide(
   A=Add(A=Distance(A=@n59, B=@n60.XYZ), B=ScalarParameter() {defaultValue=-500, parameterName=WPO.SineWobble.BlendOffset})
   B=ScalarParameter() {defaultValue=10000, parameterName=WPO.SineWobble.BlendDistance}
  )
  NamedRerouteDeclaration(None=LinearInterpolate(A=ScalarParameter() {parameterName=WPO.SineWobble.Near}, B=ScalarParameter() {parameterName=WPO.SineWobble.Far}, Alpha=Saturate(None=@n61))) {name=DistanceBasedSineWobble, nodeColor=<rgba 0 0.7187 1 1>}
  n62 = ScalarParameter() {parameterName=Debug.Time}
  n63 = MaterialFunctionCall(f3=Frac(None=ActorPositionWS())) {materialFunction=&AddComponents}
  n64 = Add(A=Time(), B=NamedRerouteDeclaration(None=Multiply(A=@n63.f3) {constB=5}) {name=PerActorRandom, nodeColor=<rgba 0.1016 1 0 1>}) {constB=0}
  NamedRerouteDeclaration(None=Multiply(
    A=StaticSwitchParameter(True=@n62, False=@n64) {parameterName=UseDebugTime}
    B=ScalarParameter() {defaultValue=0.2, group=General, parameterName=Master.Time.Speed, sortPriority=1}
   )) {name=Time.GlobalScaled_Random, nodeColor=<rgba 1 0.9609 0 1>}
  NamedRerouteDeclaration(None=TextureObject() {samplerType=SAMPLERTYPE_Grayscale, texture=&T_Smoke_Trail_Soft}) {name=Smoke.TexObj, nodeColor=<rgba 0 0.6719 1 1>}
  n65 = ScalarParameter() {defaultValue=1, group=Fresnel, parameterName=Fresnel.Blend}
  n66 = ScalarParameter() {defaultValue=1, group=Fresnel, parameterName=Fresnel.Exponent}
  n67 = VectorParameter() {parameterName=Core.Fresnel.Axis, defaultValue=<rgba 0 1 0 1>}
  n68 = LinearInterpolate(
   A=MaterialFunctionCall(Brightness=ScalarParameter() {defaultValue=1, group=Fresnel, parameterName=Fresnel.Brightness}, Exponent=@n66) {materialFunction=&MF_Fresnel_FX}.Fresnel Double Sided
   B=MaterialFunctionCall(Axis=Transform(None=@n67.RGB), Exponent=@n66) {materialFunction=&AxisAlignedFresnel}.Fresnel
   Alpha=ScalarParameter() {defaultValue=1, group=Fresnel, parameterName=Fresnel.SingleAxisFresnelAlpha}
  )
  NamedRerouteDeclaration(None=Saturate(None=LinearInterpolate(B=@n68, Alpha=@n65) {constA=1})) {name=Fresnel.Mask, nodeColor=<rgba 0 1 0.1797 1>}
  n69 = Multiply(A=NamedRerouteUsage()) {constB=1.11}
  n70 = ScalarParameter() {defaultValue=0.5, group=Noise, parameterName=Noise.UVWarp.Tiling}
  n71 = Multiply(A=NamedRerouteUsage(), B=ScalarParameter() {defaultValue=0.13, group=Noise, parameterName=Noise.UVWarp.Speed}) {constB=0.13}
  n72 = Multiply(A=@n71) {constB=-1}
  n73 = TextureSample(
   UVs=Add(
    A=Multiply(A=@n69, B=@n70)
    B=MaterialFunctionCall(X=@n72, Y=@n71) {materialFunction=&MakeFloat2}.Result
   )
   Tex=NamedRerouteUsage()
  )
  n74 = Multiply(
   A=ConstantBiasScale(None=MaterialFunctionCall(X=@n73.R, Y=@n73.R) {materialFunction=&MakeFloat2}.Result) {bias=-0.5, scale=2}
   B=ScalarParameter() {defaultValue=0.025, group=Noise, parameterName=Noise.Warp.Intensity}
  ) {constB=0}
  n75 = Multiply(
   A=NamedRerouteUsage()
   B=AppendVector(
    A=ScalarParameter() {defaultValue=1, group=Noise, parameterName=Noise.PanSpeed.X}
    B=ScalarParameter() {group=Noise, parameterName=Noise.PanSpeed.Y}
   )
  ) {constB=0}
  n76 = Add(A=@n75, B=@n74)
  n77 = Multiply(
   A=NamedRerouteUsage()
   B=MaterialFunctionCall(
    X=ScalarParameter() {defaultValue=1, group=Noise, parameterName=Noise.Tiling.X}
    Y=ScalarParameter() {defaultValue=1, group=Noise, parameterName=Noise.Tiling.Y}
   ) {materialFunction=&MakeFloat2}.Result
  )
  n78 = NamedRerouteUsage()
  n79 = NamedRerouteUsage()
  n80 = TextureSample(UVs=Add(A=Add(A=@n76, B=@n77), B=@n78), Tex=@n79) {samplerSource=SSM_Wrap_WorldGroupSettings}
  n81 = ScalarParameter() {defaultValue=1, parameterName=Noise.Power.A}
  n82 = Multiply(A=@n75, B=ScalarParameter() {defaultValue=1.27, group=Noise, parameterName=Noise.PanSpeed.LayerMultiB}) {constB=1.27}
  n83 = ConstantBiasScale(None=@n73.B) {bias=-0.5, scale=2}
  n84 = Multiply(A=ScalarParameter() {defaultValue=0.15, group=Noise, parameterName=Abberation.Intensity}) {constB=0.1}
  n85 = Add(A=@n74, B=Multiply(A=@n83, B=@n84) {constB=0.02})
  n86 = NamedRerouteUsage()
  n87 = MaterialFunctionCall(X=ScalarParameter() {defaultValue=1, group=Noise, parameterName=Noise.Tiling.X.B}, Y=Constant() {r=-0.9}) {materialFunction=&MakeFloat2}
  n88 = Add(A=Add(A=Add(A=@n82, B=@n85), B=Multiply(A=@n86, B=@n87.Result)), B=@n78)
  n89 = Multiply(A=TextureSample(UVs=@n77) {samplerSource=SSM_Wrap_WorldGroupSettings, texture=/Game/Effects/Fort_Effects/Textures/T_WispyFireNoise03}.R) {constB=0.05}
  n90 = Power(
   Base=TextureSample(UVs=Add(A=@n88, B=@n89), Tex=@n79) {samplerSource=SSM_Wrap_WorldGroupSettings}.R
   Exp=ScalarParameter() {defaultValue=1, parameterName=Noise.Power.B}
  )
  n91 = Add(
   A=Multiply(A=Divide(A=@n80.R, B=@n81), B=ScalarParameter() {defaultValue=1, parameterName=Noise.Multi.A}) {constB=0.5}
   B=Multiply(A=Divide(A=@n90), B=ScalarParameter() {defaultValue=1, parameterName=Noise.Multi.B}) {constB=0.5}
  )
  n92 = Multiply(A=@n84) {constB=-1}
  n93 = Add(
   A=Multiply(A=@n75, B=ScalarParameter() {defaultValue=1.13, group=Noise, parameterName=Noise.PanSpeed.LayerMultiC}) {constB=1.13}
   B=Add(A=@n74, B=Multiply(A=@n83, B=@n92) {constB=-0.02})
  )
  n94 = ScalarParameter() {defaultValue=1, group=Noise, parameterName=Noise.Tiling.X.C}
  n95 = Multiply(
   A=Multiply(A=NamedRerouteUsage()) {constB=-1}
   B=MaterialFunctionCall(X=Multiply(A=@n94) {constB=-1}, Y=Constant() {r=1.05}) {materialFunction=&MakeFloat2}.Result
  )
  n96 = TextureSample(UVs=Add(A=Add(A=@n93, B=@n95), B=@n78), Tex=@n79) {samplerSource=SSM_Wrap_WorldGroupSettings}
  n97 = ScalarParameter() {defaultValue=1, parameterName=Noise.Power.C}
  n98 = Multiply(A=Divide(A=Power(Base=@n96.R, Exp=@n97) {constExponent=4}), B=ScalarParameter() {defaultValue=1, parameterName=Noise.Multi.C}) {constB=1.5}
  NamedRerouteDeclaration(None=Saturate(None=Add(A=@n91, B=@n98))) {name=Mask.Noise, nodeColor=<rgba 1 0.5391 0 1>}
  n99 = ScalarParameter() {bUseCustomPrimitiveData=true, defaultValue=1, parameterName=CustomPrim_UseEndTipWPO, primitiveDataIndex=5}
  NamedRerouteDeclaration(None=Max(A=@n26, B=Multiply(A=@n27, B=@n99))) {name=Tip.Mask.WPO, nodeColor=<rgba 0.5469 1 0 1>}
  n100 = LinearInterpolate(
   A=LinearInterpolate(
    A=VectorParameter() {parameterName=DebugColorA, defaultValue=<rgba 1 0 0 1>}.RGB
    B=VectorParameter() {parameterName=DebugColorB, defaultValue=<rgba 0 0 1 1>}.RGB
    Alpha=NamedRerouteUsage()
   )
   Alpha=NamedRerouteUsage()
  )
  n101 = ScalarParameter() {defaultValue=1, group=Debug, parameterName=Debug.Brightness, sortPriority=33}
  n102 = Saturate(None=Add(A=VectorParameter() {parameterName=Airglow.Color, defaultValue=<rgba 1 0.7264 0 1>}.RGB, B=NamedRerouteUsage()))
  n103 = ScalarParameter() {defaultValue=1, parameterName=Airglow.Fresnel.Brightness}
  n104 = ScalarParameter() {defaultValue=1, parameterName=Airglow.Fresnel.Exponent}
  n105 = NamedRerouteUsage()
  n106 = NamedRerouteUsage()
  n107 = ScalarParameter() {defaultValue=10, parameterName=Airglow.WorldNoise.Speed}
  n108 = MaterialFunctionCall(
   TextureObject=NamedRerouteUsage()
   TextureSize=ScalarParameter() {defaultValue=1000, parameterName=Airglow.WorldNoise.Size}
   WorldPosition=Add(A=WorldPosition().XYZ, B=Multiply(A=@n106, B=@n107) {constB=10})
  ) {materialFunction=&WorldAlignedTexture}
  n109 = Subtract(
   A=Multiply(
    A=MaterialFunctionCall(Brightness=@n103, Exponent=@n104) {materialFunction=&MF_Fresnel_FX}.Fresnel
    B=OneMinus(None=@n105)
   )
   B=Multiply(A=ComponentMask(None=@n108.XYZ Texture) {r=true}, B=ScalarParameter() {defaultValue=0.15, parameterName=Airglow.WorldNoise.Opacity}) {constB=0.15}
  )
  n110 = Multiply(A=@n109, B=ScalarParameter() {defaultValue=0.3, parameterName=Airglow.Opacity})
  n111 = NamedRerouteUsage()
  n112 = OneMinus(None=NamedRerouteUsage())
  n113 = ShadingPathSwitch(
   Default=DepthFade(Opacity=NamedRerouteUsage(), FadeDistance=ScalarParameter() {defaultValue=25, parameterName=Opacity.DepthFade}) {fadeDistanceDefault=25}
   Mobile=NamedRerouteUsage()
  )
  n114 = OneMinus(None=NamedRerouteUsage())
  n115 = LinearInterpolate(
   A=VectorParameter() {group=Color, parameterName=Color.A, defaultValue=<rgba 1 0.7617 0.4479 1>}.RGB
   B=VectorParameter() {group=Color, parameterName=Color.B, defaultValue=<rgba 1 0.6066 0.0885 1>}.RGB
   Alpha=NamedRerouteUsage()
  )
  n116 = Add(
   A=Multiply(A=NamedRerouteUsage(), B=NamedRerouteDeclaration(None=@n115) {name=SplineColor, nodeColor=<rgba 1 0 0.0937 1>}) {constA=1}
   B=NamedRerouteUsage()
  )
  n117 = Saturate(None=NamedRerouteUsage())
  n118 = Constant()
  n119 = VectorParameter() {group=Color, parameterName=Color.Shimmer, defaultValue=<rgba 1 1 1 1>}
  n120 = Saturate(None=Multiply(A=AppendVector(A=@n117, B=@n118), B=Multiply(A=@n119.RGB, B=@n119.RGB)))
  n121 = Constant() {r=0.5}
  n122 = Constant() {r=0.1}
  n123 = Add(A=TextureCoordinate(), B=Multiply(A=NamedRerouteUsage()) {constB=0.05})
  n124 = Multiply(
   A=Multiply(
    A=Add(A=@n116, B=@n120)
    B=MaterialFunctionCall!(Width=@n121, Sharpness=@n122, Offset, Direction Switch, Compare, Input Coordinates=@n123) {materialFunction=/Engine/Functions/Engine_MaterialFunctions03/Procedurals/GeneratedBand}.Result
   )
   B=NamedRerouteUsage()
  )
  n125 = ScalarParameter() {group=General, parameterName=General.SinePulse.MinVal}
  n126 = ScalarParameter() {defaultValue=1, group=General, parameterName=General.SinePulse.MaxVal}
  n127 = ConstantBiasScale(None=Sine(None=NamedRerouteUsage()))
  n128 = ScalarParameter() {defaultValue=1, group=General, parameterName=General.FarGlow.Multiplier}
  n129 = OneMinus(None=SphereMask(
    A=WorldPosition().XYZ
    B=CameraPositionWS()
    Radius=ScalarParameter() {defaultValue=5000, group=General, parameterName=General.FarGlow.Radius}
    Hardness=ScalarParameter() {defaultValue=0.8, group=General, parameterName=General.FarGlow.Hardness}
   ))
  n130 = Multiply(
   A=ScalarParameter() {defaultValue=1, parameterName=Emissive.Multiplier}
   B=Multiply(A=LinearInterpolate(A=@n125, B=@n126, Alpha=@n127), B=LinearInterpolate(B=@n128, Alpha=@n129) {constA=1})
  )
  n131 = WorldPosition()
  n132 = CameraPositionWS()
  n133 = ScalarParameter() {defaultValue=256, parameterName=Color.Var.Radius}
  n134 = ScalarParameter() {defaultValue=1, parameterName=Color.Var.Hardness}
  n135 = Multiply(
   A=Multiply(A=Saturate(None=@n114), B=Multiply(A=@n124, B=@n130))
   B=LinearInterpolate(
    A=ScalarParameter() {defaultValue=1, parameterName=Color.Var.Min}
    B=ScalarParameter() {defaultValue=1, parameterName=Color.Var.Max}
    Alpha=SphereMask(A=@n131.XYZ, B=@n132, Radius=@n133, Hardness=@n134)
   )
  )
  n136 = GetMaterialAttributes(None=SetMaterialAttributes!(
    MaterialAttributes=SetMaterialAttributes!(MaterialAttributes, Emissive Color=@n135) {attributeSetTypes [#1]}
    Opacity=Multiply(A=@n112, B=@n113)
   ) {attributeSetTypes [#2]}) {attributeGetTypes [#1]}
  n137 = Multiply(A=Multiply(A=@n136.Emissive Color, B=@n28), B=ScalarParameter() {defaultValue=0.2, group=Pulse, parameterName=Pulse.Opacity})
  n138 = GetMaterialAttributes(None=SetMaterialAttributes(MaterialAttributes=@n136.MaterialAttributes, Emissive Color=Add(A=@n136.Emissive Color, B=@n137)) {attributeSetTypes [#1]}) {attributeGetTypes [#3]}
  n139 = NamedRerouteUsage()
  n140 = Multiply(A=NamedRerouteUsage(), B=ScalarParameter() {defaultValue=1, parameterName=WPO.SineWobble.Speed})
  n141 = Add(A=@n139, B=Multiply(A=@n140) {constB=1.11})
  n142 = ScalarParameter() {
   defaultValue=1
   desc=Higher values will make the effect more noticeable and on smaller parts of the spline at once
   parameterName=WPO.SineWobble.Contrast
  }
  n143 = Power(Base=ConstantBiasScale(None=Sine(None=@n141) {period=5.2}), Exp=@n142)
  n144 = Add(A=@n139, B=Multiply(A=@n140) {constB=1.43})
  n145 = Power(Base=ConstantBiasScale(None=Sine(None=@n144) {period=2.14}), Exp=@n142)
  n146 = Multiply(A=@n140) {constB=0.66}
  n147 = ConstantBiasScale(None=Sine(None=Add(A=@n139, B=@n146)) {period=1.72})
  n148 = Add(
   A=Add(A=ConstantBiasScale(None=@n143) {bias=-0.5, scale=2}, B=ConstantBiasScale(None=@n145) {bias=-0.5, scale=2})
   B=ConstantBiasScale(None=Power(Base=@n147, Exp=@n142)) {bias=-0.5, scale=2}
  )
  n149 = NamedRerouteUsage()
  n150 = NamedRerouteUsage()
  n151 = Add(A=@n150, B=Multiply(A=@n140) {constB=1.9867})
  n152 = ScalarParameter() {
   defaultValue=1
   desc=Higher values will make the effect more noticeable and on smaller parts of the spline at once
   parameterName=WPO.SineWobble.Contrast
  }
  n153 = Power(Base=ConstantBiasScale(None=Sine(None=@n151) {period=2.74}), Exp=@n152)
  n154 = Add(A=@n150, B=Multiply(A=@n140) {constB=1.8576})
  n155 = Power(Base=ConstantBiasScale(None=Sine(None=@n154) {period=1.44}), Exp=@n152)
  n156 = Multiply(A=@n140) {constB=0.7589}
  n157 = ConstantBiasScale(None=Sine(None=Add(A=@n150, B=@n156)) {period=1.29})
  n158 = Add(
   A=Add(A=ConstantBiasScale(None=@n153) {bias=-0.5, scale=2}, B=ConstantBiasScale(None=@n155) {bias=-0.5, scale=2})
   B=ConstantBiasScale(None=Power(Base=@n157, Exp=@n152)) {bias=-0.5, scale=2}
  )
  n159 = NamedRerouteUsage()
  n160 = NamedRerouteUsage()
  n161 = Add(A=@n160, B=Multiply(A=@n140) {constB=1.2})
  n162 = ScalarParameter() {
   defaultValue=1
   desc=Higher values will make the effect more noticeable and on smaller parts of the spline at once
   parameterName=WPO.SineWobble.Contrast
  }
  n163 = Power(Base=ConstantBiasScale(None=Sine(None=@n161) {period=1.52}), Exp=@n162)
  n164 = Add(A=@n160, B=Multiply(A=@n140) {constB=0.95})
  n165 = Power(Base=ConstantBiasScale(None=Sine(None=@n164) {period=0.91}), Exp=@n162)
  n166 = Multiply(A=@n140) {constB=0.79}
  n167 = ConstantBiasScale(None=Sine(None=Add(A=@n160, B=@n166)) {period=2.45})
  n168 = Add(
   A=Add(A=ConstantBiasScale(None=@n163) {bias=-0.5, scale=2}, B=ConstantBiasScale(None=@n165) {bias=-0.5, scale=2})
   B=ConstantBiasScale(None=Power(Base=@n167, Exp=@n162)) {bias=-0.5, scale=2}
  )
  n169 = NamedRerouteUsage()
  n170 = ScalarParameter() {defaultValue=1, parameterName=WPO.SineWobble.TipMulti}
  n171 = ScalarParameter() {bUseCustomPrimitiveData=true, defaultValue=1, parameterName=CustomPrim_UseEndTipWPO, primitiveDataIndex=5}
  n172 = Multiply(
   A=MaterialFunctionCall(X=Multiply(A=@n148, B=@n149), Y=Multiply(A=@n158, B=@n159), Z=Multiply(A=@n168, B=@n169)) {materialFunction=&MakeFloat3}.Result
   B=LinearInterpolate(A=LinearInterpolate(A=@n170, Alpha=@n171) {constA=1}, Alpha=NamedRerouteUsage()) {constA=1}
  )
  n173 = NamedRerouteUsage()
  n174 = SetMaterialAttributes!(
   MaterialAttributes=@n138.MaterialAttributes
   World Position Offset=Add(A=@n138.World Position Offset, B=LinearInterpolate(A=@n172, Alpha=@n173) {constB=0}) {constB=0}
  ) {attributeSetTypes [#3]}
  n175 = GetMaterialAttributes(None=SetMaterialAttributes!(MaterialAttributes=@n174, Emissive Color=Multiply(A=@n102, B=@n109), Opacity=Multiply(A=@n110, B=@n111)) {attributeSetTypes [#1, #2]}) {attributeGetTypes [#3]}
  n176 = VertexNormalWS()
  n177 = CameraPositionWS()
  n178 = WorldPosition()
  n179 = Divide(
   A=Add(A=Distance(A=@n177, B=@n178.XYZ), B=ScalarParameter() {defaultValue=-500, parameterName=Airglow.WPO.BlendOffset})
   B=ScalarParameter() {defaultValue=10000, parameterName=Airglow.WPO.BlendDistance}
  )
  n180 = MaterialFunctionCall(Float2=NamedRerouteUsage()) {materialFunction=&BreakOutFloat2Components}
  n181 = TextureCoordinate()
  n182 = Multiply(
   A=LinearInterpolate(
    A=ScalarParameter() {defaultValue=100, parameterName=Airglow.WPORadius.Near}
    B=ScalarParameter() {defaultValue=100, parameterName=Airglow.WPORadius.Far}
    Alpha=Saturate(None=@n179)
   )
   B=LinearInterpolate(A=@n180.G, B=@n180.G, Alpha=ComponentMask(None=@n181) {r=true})
  )
  n183 = SetMaterialAttributes!(MaterialAttributes=@n175.MaterialAttributes, World Position Offset=Add(A=@n175.World Position Offset, B=Multiply(A=@n176, B=@n182))) {attributeSetTypes [#3]}
  n184 = OneMinus(None=VertexColor().B)
  n185 = GetMaterialAttributes(None=StaticSwitchParameter(True=BlendMaterialAttributes(A=@n183, B=@n174, Alpha=@n184), False=@n174) {defaultValue=true, parameterName=UseAirglowVertPaintAndMesh}) {attributeGetTypes [#1]}
  StaticSwitchParameter(
   True=SetMaterialAttributes!(MaterialAttributes, Emissive Color=Multiply(A=@n100, B=@n101)) {attributeSetTypes [#1]}
   False=SetMaterialAttributes(
    MaterialAttributes=@n185.MaterialAttributes
    Emissive Color=MaterialFunctionCall(Emissive=@n185.Emissive Color) {materialFunction=&MF_FX_EmissiveAdjustments}.Result
   ) {attributeSetTypes [#1]}
  ) {group=Debug, parameterName=DebugView}
 ]
}
