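# Masked, lit, responsive-AA material for gold airborne ribbon sparkles in GrillMoon.
# World position offset builds sprites that shrink near the camera and adds sine wobble.
# Opacity layers two scales of thresholded, counter-scrolling white noise, shaped by radial gradient, camera-distance fades and depth fade.
# Gold color gets twinkle and distance-based emissive multipliers, with cloud-noise masks and shared time offset by actor position to vary instances.
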Material /GrillMoon/Assets/TechArt/GoldRibbons/M_RibbonAirSparkles_Grillmoon_01
graph {
 properties {bAutomaticallySetUsageInEditor=false, bEnableResponsiveAA=true, blendMode=BLEND_Masked, usedShadingModels=Default Lit}
 nodes [
  n1 = ScalarParameter() {defaultValue=500, desc=Will Convert to Custom Prim Data when working correctly, parameterName=WPOScale, primitiveDataIndex=2}
  n2 = WorldPosition()
  n3 = CameraPositionWS()
  n4 = SmoothStep(Value=SphereMask(A=@n2.XYZ, B=@n3, Radius=ScalarParameter() {defaultValue=1000, parameterName=WPO.ScaleStartDistance}) {attenuationRadius=1000, hardnessPercent=50}) {constMax=0.7}
  n5 = Multiply(A=NamedRerouteUsage(), B=ScalarParameter() {defaultValue=1, parameterName=WPO.SineWobble.Speed})
  n6 = Sine(None=Multiply(A=@n5) {constB=1.11}) {period=5.2}
  n7 = Sine(None=Multiply(A=@n5) {constB=1.9867}) {period=2.74}
  n8 = Sine(None=Multiply(A=@n5) {constB=1.2}) {period=1.52}
  Add(
   A=MaterialFunctionCall!(Center Location, 0-1 UVs, XY scale=LinearInterpolate(A=@n1, Alpha=@n4) {constB=0}, OverridePivot, "Normals (Optional)") {materialFunction=&Sprite}.World position offset
   B=Multiply(
    A=MaterialFunctionCall(X=@n6, Y=@n7, Z=@n8) {materialFunction=&MakeFloat3}.Result
    B=ScalarParameter() {parameterName=WPO.SineWobble}
   )
  )
  n9 = SmoothStep(Value=SphereMask(A=@n2.XYZ, B=@n3, Radius=ScalarParameter() {defaultValue=1000, parameterName=Opacity.FadeStartDistance}) {attenuationRadius=1000, hardnessPercent=0.5}) {constMax=0.5}
  n10 = MaterialFunctionCall(Density=Constant() {r=4}) {materialFunction=&RadialGradientExponential}
  n11 = Multiply(A=TextureCoordinate())
  n12 = Multiply(A=NamedRerouteUsage()) {constB=0.1}
  n13 = TextureObject() {samplerType=SAMPLERTYPE_LinearGrayscale, texture=/Game/Effects/Fort_Effects/Textures/Tile/T_WhiteNoise_02_M}
  n14 = Multiply(A=@n11) {constB=0.75}
  n15 = Multiply(A=@n12, B=Constant2Vector() {g=-1, r=1})
  n16 = Multiply(
   A=TextureSample(UVs=Add(A=@n11, B=@n12), Tex=@n13) {samplerType=SAMPLERTYPE_LinearGrayscale, texture=&T_WhiteNoise_02_M}.R
   B=TextureSample(UVs=Add(A=@n14, B=@n15), Tex=@n13) {samplerType=SAMPLERTYPE_LinearGrayscale, texture=&T_WhiteNoise_02_M}.R
  )
  n17 = Saturate(None=Multiply(A=Power(Base=@n16) {constExponent=4}) {constB=2})
  n18 = MaterialFunctionCall(Float3=NamedRerouteUsage()) {materialFunction=&BreakOutFloat3Components}
  n19 = Multiply(A=TextureCoordinate()) {constB=2}
  n20 = Multiply(A=@n19) {constB=0.66}
  n21 = Multiply(A=@n12, B=Constant2Vector() {g=1, r=-1})
  n22 = Multiply(
   A=TextureSample(UVs=Add(A=@n19, B=@n12), Tex=@n13) {samplerType=SAMPLERTYPE_LinearGrayscale, texture=&T_WhiteNoise_02_M}.R
   B=TextureSample(UVs=Add(A=@n20, B=@n21), Tex=@n13) {samplerType=SAMPLERTYPE_LinearGrayscale, texture=&T_WhiteNoise_02_M}.R
  )
  n23 = Saturate(None=Multiply(A=Power(Base=@n22) {constExponent=4}) {constB=2})
  n24 = Max(A=Multiply(A=Round(None=@n17), B=@n18.R), B=Multiply(A=@n18.G, B=Round(None=@n23)))
  DepthFade(Opacity=Multiply(A=Multiply(A=@n24, B=@n10.RadialGradientExponential), B=LinearInterpolate(Alpha=@n9) {constA=1, constB=0}))
  n25 = VectorParameter() {parameterName=Color.A, defaultValue=<rgba 1 0.7852 0.3438 1>}
  LinearInterpolate(B=@n25.RGB, Alpha=@n24)
  n26 = Time()
  n27 = ScalarParameter() {defaultValue=0.1, parameterName=Master.Speed}
  n28 = MaterialFunctionCall(f3=Frac(None=ActorPositionWS())) {materialFunction=&AddComponents}
  NamedRerouteDeclaration(None=Add(A=Multiply(A=@n26, B=@n27) {constB=0.1}, B=Multiply(A=@n28.f3) {constB=5})) {name=Time.Scaled, nodeColor=<rgba 1 0.0234 0 1>}
  n29 = WorldPosition()
  n30 = CameraPositionWS()
  n31 = VectorParameter() {parameterName=PerChannelNearFade, channelNames {a=Hardness, b="", g="", r=""}, defaultValue=<rgba 1000 750 500 50>}
  n32 = TextureSample(Tex=NamedRerouteUsage())
  n33 = ScalarParameter() {parameterName=PerChannelFadeVariation}
  n34 = MaterialFunctionCall(Float3=Add(A=@n31, B=Multiply(A=@n32.RGB, B=@n33))) {materialFunction=&BreakOutFloat3Components}
  n35 = SphereMask(A=@n29.XYZ, B=@n30, Radius=@n34.R, Hardness=@n31.Hardness) {attenuationRadius=1000, hardnessPercent=0.5}
  n36 = SphereMask(A=@n29.XYZ, B=@n30, Radius=@n34.G, Hardness=@n31.Hardness) {attenuationRadius=1000, hardnessPercent=0.5}
  n37 = SphereMask(A=@n29.XYZ, B=@n30, Radius=@n34.B, Hardness=@n31.Hardness) {attenuationRadius=1000, hardnessPercent=0.5}
  NamedRerouteDeclaration(None=MaterialFunctionCall(X=OneMinus(None=@n35), Y=OneMinus(None=@n36), Z=OneMinus(None=@n37)) {materialFunction=&MakeFloat3}.Result) {name=PackedDistanceFades, nodeColor=<rgba 0.2656 0 1 1>}
  NamedRerouteDeclaration(None=TextureObject() {texture=&/T_Cloud_Noise_GrillMoon}) {name=Noise.A.Obj, nodeColor=<rgba 0 0.6719 1 1>}
  n38 = Multiply(
   A=TextureCoordinate()
   B=MaterialFunctionCall(
    X=ScalarParameter() {defaultValue=1, parameterName=MaskedNoise.Tiling.X}
    Y=ScalarParameter() {defaultValue=1, parameterName=MaskedNoise.Tiling.Y}
   ) {materialFunction=&MakeFloat2}.Result
  )
  n39 = NamedRerouteUsage()
  n40 = Multiply(A=@n39) {constB=-0.15}
  n41 = Multiply(A=@n39) {constB=0.25}
  n42 = TextureSample(
   UVs=Add(A=@n38, B=MaterialFunctionCall(X=@n40, Y=@n41) {materialFunction=&MakeFloat2}.Result)
   Tex=NamedRerouteUsage()
  ) {samplerType=SAMPLERTYPE_LinearColor}
  n43 = Multiply(A=OneMinus(None=@n42.B)) {constB=0.1}
  n44 = Add(A=@n38, B=MaterialFunctionCall(Y=Multiply(A=@n39)) {materialFunction=&MakeFloat2}.Result)
  n45 = Add(
   A=@n38
   B=MaterialFunctionCall(X=Multiply(A=@n39) {constB=-0.1}, Y=Multiply(A=@n39) {constB=0.75}) {materialFunction=&MakeFloat2}.Result
  )
  n46 = Multiply(A=@n42.B) {constB=0.1}
  n47 = Add(
   A=TextureSample(UVs=Add(A=@n44, B=@n43), Tex=NamedRerouteUsage()) {samplerType=SAMPLERTYPE_LinearColor}.R
   B=TextureSample(UVs=Add(A=@n45, B=@n46), Tex=NamedRerouteUsage()) {samplerType=SAMPLERTYPE_LinearColor}.G
  )
  n48 = Multiply(A=@n42.B) {constB=0.3}
  n49 = Saturate(None=Divide(A=Add(A=@n47, B=@n48)) {constB=3})
  NamedRerouteDeclaration(None=@n49) {name=Mask.Noise.UVWarp, nodeColor=<rgba 1 0.5391 0 1>}
  n50 = TextureCoordinate()
  n51 = NamedRerouteUsage()
  n52 = Multiply(A=@n51, B=ScalarParameter() {defaultValue=3, parameterName=Twinkle.Mask.SineSpeed}) {constB=3}
  n53 = Multiply(
   A=MaterialFunctionCall(X=Sine(None=@n52) {period=1.17}, Y=Sine(None=@n52) {period=1.63}) {materialFunction=&MakeFloat2}.Result
   B=ScalarParameter() {defaultValue=0.4, parameterName=Twinkle.Mask.SineIntensity}
  ) {constB=0.4}
  n54 = Multiply(A=@n51, B=ScalarParameter() {defaultValue=2, parameterName=Twinkle.Mask.SpeedBase}) {constB=5}
  n55 = Constant2Vector() {g=-0.5, r=-1}
  n56 = TextureSample(UVs=Add(A=Add(A=@n50, B=@n53) {constB=0.5}, B=Multiply(A=@n54, B=@n55)), Tex=NamedRerouteUsage()) {samplerType=SAMPLERTYPE_LinearColor}
  n57 = ConstantBiasScale(None=Sine(None=@n54))
  n58 = LinearInterpolate(A=@n56.R, B=@n56.R, Alpha=SmoothStep(Value=Saturate(None=@n57)) {constMax=0.8, constMin=0.2})
  n59 = Power(Base=Saturate(None=Subtract(A=@n58, B=@n56.B)), Exp=ScalarParameter() {defaultValue=8, parameterName=Twinkle.Mask.Contrast}) {constExponent=8}
  n60 = ScalarParameter() {defaultValue=30, parameterName=Twinkle.Mask.Multiplier}
  NamedRerouteDeclaration(None=Saturate(None=Multiply(A=@n59, B=@n60) {constB=30})) {name=TwinkleMask, nodeColor=<rgba 1 0.3516 0 1>}
  n61 = Power(Base=@n49) {constExponent=6}
  NamedRerouteDeclaration(None=Saturate(None=Multiply(A=@n61) {constB=50})) {name=Mask.Noise, nodeColor=<rgba 1 0.5391 0 1>}
  n62 = ScalarParameter() {defaultValue=5, parameterName=Twinkle.Brightness}
  n63 = NamedRerouteUsage()
  n64 = TextureCoordinate()
  n65 = Multiply(A=Subtract(A=NamedRerouteUsage()) {constB=0.5})
  n66 = Multiply(A=NamedRerouteUsage()) {constB=3}
  n67 = Multiply(A=MaterialFunctionCall(X=Sine(None=@n66) {period=1.17}, Y=Sine(None=@n66) {period=1.63}) {materialFunction=&MakeFloat2}.Result) {constB=0.15}
  n68 = LinearInterpolate(
   A=ScalarParameter() {defaultValue=1, parameterName=Emissive.Multi.Min}
   B=ScalarParameter() {defaultValue=1, parameterName=Emissive.Multi.Max}
   Alpha=MaterialFunctionCall(UVs=Add(A=@n64, B=@n65), CenterPosition=Add(A=@n67) {constB=0.5}, Density=Constant() {r=1.5}) {materialFunction=&RadialGradientExponential}.RadialGradientExponential
  )
  n69 = Saturate(None=@n25.RGB)
  n70 = LinearInterpolate(
   A=Multiply(A=ScalarParameter() {defaultValue=1, parameterName=Emissive.Multi.Distant}, B=@n69)
   B=@n69
   Alpha=SphereMask(
    A=WorldPosition().XYZ
    B=CameraPositionWS()
    Radius=ScalarParameter() {defaultValue=2000, parameterName=Emissive.DistantBoostRadius}
    Hardness=ScalarParameter() {defaultValue=1, parameterName=Emissive.DistantBoostHardness}
   ) {attenuationRadius=1000, hardnessPercent=1}
  )
  Reroute(None=Multiply(A=LinearInterpolate(B=@n62, Alpha=@n63) {constA=1}, B=Multiply(A=@n68, B=@n70)))
  Reroute(None=@n25.RGB)
 ]
}
