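# Additive unlit material for Grillmoon gold ribbon ends, drawing a camera-facing sprite with a soft, animated swirling glow.
# Two panning layers of swirl-mask noise, warped by a noise texture, blend two gold-orange colours with optional per-channel chromatic aberration.
# A warped radial gradient with depth fade shapes opacity, and camera-distance sphere masks shrink and fade the sprite up close.
# Actor position offsets time so each instance animates differently.
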Material /GrillMoon/Assets/TechArt/GoldRibbons/M_RibbonEnd_Grillmoon_01
graph {
 properties {bAutomaticallySetUsageInEditor=false, blendMode=BLEND_Additive, shadingModel=MSM_Unlit, usedShadingModels=Unlit}
 nodes [
  n1 = WorldPosition()
  n2 = CameraPositionWS()
  n3 = SphereMask(A=@n1.XYZ, B=@n2, Radius=ScalarParameter() {defaultValue=1000, parameterName=WPO.ScaleStartDistance}) {attenuationRadius=1000, hardnessPercent=0.5}
  MaterialFunctionCall!(
   Center Location
   0-1 UVs
   XY scale=LinearInterpolate(A=ScalarParameter() {defaultValue=500, parameterName=WPO.Scale}, Alpha=SmoothStep(Value=@n3) {constMax=0.7}) {constB=0}
   OverridePivot
   "Normals (Optional)"
  ) {materialFunction=&Sprite}
  n4 = Multiply(A=TextureCoordinate()) {constB=1.11}
  n5 = ScalarParameter() {defaultValue=0.5, parameterName=Noise.Tiling}
  n6 = Multiply(A=Time(), B=ScalarParameter() {defaultValue=1, parameterName=Master.Speed})
  n7 = Frac(None=ActorPositionWS())
  n8 = Multiply(A=Frac(None=MaterialFunctionCall(f3=@n7) {materialFunction=&AddComponents}.f3)) {constB=123}
  n9 = Multiply(
   A=NamedRerouteDeclaration(None=Add(A=@n6, B=@n8)) {name=Time.PerActorRandom, nodeColor=<rgba 0.125 1 0 1>}
   B=ScalarParameter() {defaultValue=0.13, parameterName=Noise.WarpSpeed}
  ) {constB=0.13}
  n10 = Multiply(A=@n9) {constB=-1}
  n11 = TextureSample(UVs=Add(
    A=Multiply(A=@n4, B=@n5)
    B=MaterialFunctionCall(X=@n10, Y=@n9) {materialFunction=&MakeFloat2}.Result
   )) {texture=&/T_NoisePack_01_GrillMoon}
  n12 = Multiply(
   A=ConstantBiasScale(None=MaterialFunctionCall(X=@n11.R, Y=@n11.R) {materialFunction=&MakeFloat2}.Result) {bias=-0.5, scale=2}
   B=ScalarParameter() {defaultValue=0.025, parameterName=Noise.WarpIntensity}
  ) {constB=0}
  n13 = Add(A=TextureCoordinate(), B=Multiply(A=@n12) {constB=3})
  n14 = Constant() {r=0.44}
  n15 = SphereMask(A=@n1.XYZ, B=@n2, Radius=ScalarParameter() {defaultValue=1000, parameterName=Opacity.FadeStartDistance}) {attenuationRadius=1000, hardnessPercent=0.5}
  Multiply(
   A=DepthFade(
    Opacity=MaterialFunctionCall(UVs=@n13, Radius=@n14) {materialFunction=&RadialGradientExponential}.RadialGradientExponential
    FadeDistance=ScalarParameter() {defaultValue=50, parameterName=DepthFade.Distance}
   )
   B=LinearInterpolate(Alpha=SmoothStep(Value=@n15) {constMax=0.5}) {constA=1, constB=0}
  )
  n16 = Multiply(A=NamedRerouteUsage()) {constB=0.05}
  n17 = MaterialFunctionCall(X=Multiply(A=@n16) {constB=-0.8}, Y=Multiply(A=@n16) {constB=-0.74}) {materialFunction=&MakeFloat2}
  n18 = Multiply(A=Multiply(A=TextureCoordinate(), B=ScalarParameter() {defaultValue=0.5, parameterName=Noise.Tiling})) {constB=0.81}
  n19 = Add(A=@n18, B=@n12)
  n20 = TextureObject() {samplerType=SAMPLERTYPE_LinearGrayscale, texture=/Game/Global/Textures/Default/Flowmaps/T_StormSwirlMask}
  n21 = ConstantBiasScale(None=@n11.B) {bias=-0.5, scale=2}
  n22 = Multiply(A=ScalarParameter() {defaultValue=0.15, parameterName=Abberation.Intensity}) {constB=0.1}
  n23 = Add(A=@n12, B=Multiply(A=@n21, B=@n22) {constB=0.02})
  n24 = Add(A=@n18, B=@n23)
  n25 = Add(A=@n12, B=Multiply(A=@n21, B=Multiply(A=@n22) {constB=-1}) {constB=-0.02})
  n26 = Add(A=@n18, B=@n25)
  n27 = MaterialFunctionCall(
   X=TextureSample(UVs=Add(A=@n17.Result, B=@n19), Tex=@n20).R
   Y=TextureSample(UVs=Add(A=@n17.Result, B=@n24), Tex=@n20).R
   Z=TextureSample(UVs=Add(A=@n17.Result, B=@n26), Tex=@n20).R
  ) {materialFunction=&MakeFloat3}
  n28 = ScalarParameter() {defaultValue=1.5, parameterName=Noise.Contrast}
  n29 = Multiply(A=TextureCoordinate(), B=ScalarParameter() {defaultValue=0.5, parameterName=Noise.Tiling})
  n30 = Add(A=@n29, B=@n12)
  n31 = TextureObject() {samplerType=SAMPLERTYPE_LinearGrayscale, texture=&T_StormSwirlMask}
  n32 = Add(A=@n29, B=@n23)
  n33 = Add(A=@n29, B=@n25)
  n34 = MaterialFunctionCall(
   X=TextureSample(UVs=Add(A=@n16, B=@n30), Tex=@n31).R
   Y=TextureSample(UVs=Add(A=@n16, B=@n32), Tex=@n31).R
   Z=TextureSample(UVs=Add(A=@n16, B=@n33), Tex=@n31).R
  ) {materialFunction=&MakeFloat3}
  n35 = Divide(A=Add(A=Power(Base=@n27.Result, Exp=@n28), B=Power(Base=@n34.Result, Exp=@n28)))
  n36 = ComponentMask(None=@n35) {r=true}
  n37 = LinearInterpolate(
   A=VectorParameter() {parameterName=Color.B, defaultValue=<rgba 1 0.5615 0 1>}.RGB
   B=VectorParameter() {parameterName=Color.A, defaultValue=<rgba 1 0.7852 0.3438 1>}.RGB
   Alpha=@n36
  )
  n38 = LinearInterpolate(A=Multiply(A=@n37, B=@n36), B=Multiply(A=@n37, B=@n35), Alpha=ScalarParameter() {parameterName=Abberation.Amount}) {constAlpha=0}
  Reroute(None=Multiply(A=Saturate(None=@n38), B=ScalarParameter() {defaultValue=1, parameterName=Emissive.Multi}))
 ]
}
